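# MF_DepthCull is a material function that zeroes opacity for pixels lying behind whatever is written to the custom depth buffer.
# It compares the reciprocals of scaled CustomDepth and PixelDepth, with a small bias against equal-depth pixels culling themselves.
# Ceil and Clamp turn the difference into a hard 0/1 mask that lerps OpacityIn to zero for the Opacity output.
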MaterialFunction /WildEstate/TechArt/MF_DepthCull
graph {
 properties {previewBlendMode=BLEND_Translucent}
 nodes [
  n1 = Constant() {r=0.5}
  n2 = Constant() {r=8192}
  n3 = ComponentMask(None=SceneTexture() {sceneTextureId=PPI_CustomDepth}.Color) {r=true}
  n4 = Divide(A=@n2, B=PixelDepth().R) {constA=8192}
  n5 = Ceil(None=Subtract(A=Divide(A=@n2, B=@n3) {constA=8192}, B=Add(A=@n4) {constB=0.0001}))
  FunctionOutput(None=LinearInterpolate(A=FunctionInput(Preview=@n1) {bUsePreviewValueAsDefault=true, inputName=OpacityIn, inputType=FunctionInput_Scalar}, Alpha=Clamp(None=@n5)) {constA=0.5, constB=0}) {outputName=Opacity}
 ]
}
